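; Daisy lily pad foliage building actor whose construction script draws an optional debug visualization of a jump radius.
; When the debug jump radius flag is set, it draws an amber circle of the configured debug radius.
; The circle lies horizontally, raised slightly above the static mesh component's world location, and persists only briefly.
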(fn ConstructionScript
  (if (Variables|Default|GetDebugJumpRadius)
    (Rendering|Debug|DrawDebugCircle (+ (Transformation|GetWorldLocation (Variables|Debug|GetStaticMeshComponent)) "0,0,64.000000") (Variables|Default|GetDebugCircleRadius) 40 "(R=1.000000,G=0.744782,B=0.000000,A=1.000000)" 0.2 2.0 "0.000000,1.000000,0.000000" "1.000000,0.000000,0.000000")))
